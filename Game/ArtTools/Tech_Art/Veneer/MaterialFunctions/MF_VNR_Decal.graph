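# Material function for the Veneer tools that layers a tinted decal texture over incoming material attributes, toggled by a static switch.
# Texture alpha, input Alpha and an opacity parameter form the mask, clipped by a SphereMask on centered unscaled UVs.
# That mask blends the decal into BaseColor and blends min/max-parameter remapped values into Metallic, Specular and Roughness.
# Named reroutes carry the original material, mask and decal result.
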MaterialFunction /Game/ArtTools/Tech_Art/Veneer/MaterialFunctions/MF_VNR_Decal
graph {nodes [
  FunctionOutput(None=StaticSwitchParameter(True=NamedRerouteUsage(), False=NamedRerouteUsage()) {group=Veneer, parameterName=Enable VNR Decal})
  NamedRerouteDeclaration(None=FunctionInput() {bUsePreviewValueAsDefault=true, inputType=FunctionInput_MaterialAttributes}) {name=Original Material, nodeColor=<rgba 0.0547 0 1 1>}
  n1 = GetMaterialAttributes(None=NamedRerouteUsage()) {attributeGetTypes [
    #1
    #2
    #3
    #4
   ]}
  n2 = VectorParameter() {group=Veneer, parameterName=VNR_TintColor, defaultValue=<rgba 1 1 1 1>}
  n3 = TextureSampleParameter2D(UVs=FunctionInput(Preview=TextureCoordinate()) {bUsePreviewValueAsDefault=true, inputName=UVs, inputType=FunctionInput_Vector2, sortPriority=10}) {group=Veneer, parameterName=VNR_Texture, texture=/Game/Global/Base4Layers_2022/Textures/T-GraffitiSample_D}
  n4 = Multiply(A=@n3.A, B=FunctionInput() {bUsePreviewValueAsDefault=true, inputName=Alpha, inputType=FunctionInput_Scalar, sortPriority=20})
  n5 = ScalarParameter() {defaultValue=1, group=Veneer, parameterName=VNR_Opacity, sliderMax=1}
  n6 = FunctionInput(Preview=TextureCoordinate()) {bUsePreviewValueAsDefault=true, inputName=UV Unscaled, inputType=FunctionInput_Vector2, sortPriority=15}
  n7 = MaterialFunctionCall(Float2=Abs(None=Subtract(A=@n6) {constB=0.5})) {materialFunction=&BreakOutFloat2Components}
  n8 = Multiply(A=Saturate(None=Multiply(A=@n4, B=@n5)), B=SphereMask(A=Max(A=@n7.R, B=@n7.R), B=Constant()) {attenuationRadius=0.5})
  n9 = MaterialFunctionCall(X=@n1.Metallic, Y=@n1.Metallic, Z=@n1.Metallic) {materialFunction=&MakeFloat3}
  n10 = LinearInterpolate(
   A=MaterialFunctionCall(
    X=ScalarParameter() {group=Veneer, parameterName=VNR_Metallic_Min, sliderMax=1}
    Y=ScalarParameter() {group=Veneer, parameterName=VNR_Specular_Min, sliderMax=1}
    Z=ScalarParameter() {group=Veneer, parameterName=VNR_Roughness_Min, sliderMax=1}
   ) {materialFunction=&MakeFloat3}.Result
   B=MaterialFunctionCall(
    X=ScalarParameter() {defaultValue=1, group=Veneer, parameterName=VNR_Metallic_Max, sliderMax=1}
    Y=ScalarParameter() {defaultValue=1, group=Veneer, parameterName=VNR_Specular_Max, sliderMax=1}
    Z=ScalarParameter() {defaultValue=1, group=Veneer, parameterName=VNR_Roughness_Max, sliderMax=1}
   ) {materialFunction=&MakeFloat3}.Result
   Alpha=@n9.Result
  )
  n11 = MaterialFunctionCall(Float3=LinearInterpolate(A=@n9.Result, B=Saturate(None=@n10), Alpha=NamedRerouteUsage())) {materialFunction=&BreakOutFloat3Components}
  n12 = SetMaterialAttributes!(
   MaterialAttributes=@n1.MaterialAttributes
   BaseColor=LinearInterpolate(A=@n1.BaseColor, B=Multiply(A=@n2.RGB, B=@n3.RGB), Alpha=NamedRerouteDeclaration(None=@n8) {name=Mask, nodeColor=<rgba 0.2422 0 1 1>})
   Metallic=@n11.R
   Specular=@n11.R
   Roughness=@n11.R
  ) {attributeSetTypes [
    #1
    #2
    #3
    #4
   ]}
  NamedRerouteDeclaration(None=@n12) {name=Decal Material, nodeColor=<rgba 0 0.8934 1 1>}
 ]}
