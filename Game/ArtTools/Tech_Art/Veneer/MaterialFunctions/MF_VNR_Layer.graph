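# Veneer material function that overlays a tinted, textured decal layer on incoming material attributes, toggled by the Enable VNR Layer static switch.
# Base color alpha, Alpha input, mask texture, opacity and a UV-distance SphereMask form the blend mask.
# Min/max parameters remap the SMap's green channel, but Metallic, Specular and Roughness all receive the blend's red component.
# Normals blend by mask, then combine with transition normals derived from the mask texture as heightmap.
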MaterialFunction /Game/ArtTools/Tech_Art/Veneer/MaterialFunctions/MF_VNR_Layer
graph {nodes [
  FunctionOutput(None=StaticSwitchParameter(True=NamedRerouteUsage(), False=NamedRerouteUsage()) {group=Veneer, parameterName=Enable VNR Layer})
  n1 = TextureSampleParameter2D(UVs=NamedRerouteUsage()) {group=Veneer, parameterName=VNR_Texture_Normal, samplerType=SAMPLERTYPE_Normal, texture=&FlatNormalBlue}
  n2 = GetMaterialAttributes(None=NamedRerouteUsage()) {attributeGetTypes [
    #1
    #2
    #3
    #4
    #5
   ]}
  MaterialFunctionCall(BaseNormal=@n2.Normal, AdditionalNormal=@n1.RGB) {materialFunction=&BlendAngleCorrectedNormals}
  NamedRerouteDeclaration(None=FunctionInput() {bUsePreviewValueAsDefault=true, inputType=FunctionInput_MaterialAttributes}) {name=Original Material, nodeColor=<rgba 0.0547 0 1 1>}
  n3 = VectorParameter() {group=Veneer, parameterName=VNR_TintColor, defaultValue=<rgba 1 1 1 1>}
  n4 = TextureSampleParameter2D(UVs=NamedRerouteDeclaration(None=TextureCoordinate()) {name=Mesh UVs, nodeColor=<rgba 1 0 0.4453 1>}) {group=Veneer, parameterName=VNR_Texture_BaseColor, texture=/Game/Global/Base4Layers_2022/Textures/T-GraffitiSample_D}
  n5 = Multiply(A=@n4.A, B=FunctionInput() {bUsePreviewValueAsDefault=true, inputName=Alpha, inputType=FunctionInput_Scalar, sortPriority=20})
  n6 = Saturate(None=NamedRerouteDeclaration(None=FunctionInput() {bUsePreviewValueAsDefault=true, inputName=UVs, inputType=FunctionInput_Vector2, sortPriority=10}) {name=Projection UVs, nodeColor=<rgba 0 0.0156 1 1>})
  n7 = TextureObjectParameter() {parameterName=VNR_Texture_Mask, samplerType=SAMPLERTYPE_Masks, texture=&T_WhiteMask}
  n8 = Multiply(A=TextureSample(UVs=@n6, Tex=@n7).R, B=ScalarParameter() {defaultValue=1, group=Veneer, parameterName=VNR_Opacity, sliderMax=1})
  n9 = FunctionInput(Preview=TextureCoordinate()) {bUsePreviewValueAsDefault=true, inputName=UV Unscaled, inputType=FunctionInput_Vector2, sortPriority=15}
  n10 = MaterialFunctionCall(Float2=Abs(None=Subtract(A=@n9) {constB=0.5})) {materialFunction=&BreakOutFloat2Components}
  n11 = Multiply(A=Saturate(None=Multiply(A=@n5, B=@n8)), B=SphereMask(A=Max(A=@n10.R, B=@n10.R), B=Constant()) {attenuationRadius=0.5})
  n12 = TextureSampleParameter2D(UVs=NamedRerouteUsage()) {
   group=Veneer
   parameterName=VNR_Texture_SMap
   samplerType=SAMPLERTYPE_Masks
   texture=&T_Fortnite_Default_S
  }
  n13 = LinearInterpolate(
   A=MaterialFunctionCall(
    X=ScalarParameter() {group=Veneer, parameterName=VNR_Metallic_Min, sliderMax=1}
    Y=ScalarParameter() {group=Veneer, parameterName=VNR_Specular_Min, sliderMax=1}
    Z=ScalarParameter() {group=Veneer, parameterName=VNR_Roughness_Min, sliderMax=1}
   ) {materialFunction=&MakeFloat3}.Result
   B=MaterialFunctionCall(
    X=ScalarParameter() {defaultValue=1, group=Veneer, parameterName=VNR_Metallic_Max, sliderMax=1}
    Y=ScalarParameter() {defaultValue=1, group=Veneer, parameterName=VNR_Specular_Max, sliderMax=1}
    Z=ScalarParameter() {defaultValue=1, group=Veneer, parameterName=VNR_Roughness_Max, sliderMax=1}
   ) {materialFunction=&MakeFloat3}.Result
   Alpha=MaterialFunctionCall(X=@n12.G, Y=@n12.G, Z=@n12.G) {materialFunction=&MakeFloat3}.Result
  )
  n14 = NamedRerouteUsage()
  n15 = MaterialFunctionCall(Float3=LinearInterpolate(
    A=MaterialFunctionCall(X=@n2.Metallic, Y=@n2.Metallic, Z=@n2.Metallic) {materialFunction=&MakeFloat3}.Result
    B=Saturate(None=@n13)
    Alpha=@n14
   )) {materialFunction=&BreakOutFloat3Components}
  n16 = NamedRerouteUsage()
  n17 = SetMaterialAttributes!(
   MaterialAttributes=@n2.MaterialAttributes
   BaseColor=LinearInterpolate(A=@n2.BaseColor, B=Multiply(A=@n3.RGB, B=@n4.RGB), Alpha=NamedRerouteDeclaration(None=@n11) {name=Mask, nodeColor=<rgba 0.2422 0 1 1>})
   Metallic=@n15.R
   Specular=@n15.R
   Roughness=@n15.R
   Normal=MaterialFunctionCall(BaseNormal=LinearInterpolate(A=@n2.Normal, B=@n1.RGB, Alpha=@n14), AdditionalNormal=@n16) {materialFunction=&BlendAngleCorrectedNormals}.Result
  ) {attributeSetTypes [
    #1
    #2
    #3
    #4
    #5
   ]}
  NamedRerouteDeclaration(None=@n17) {name=Decal Material, nodeColor=<rgba 0 0.8934 1 1>}
  NamedRerouteDeclaration(None=MaterialFunctionCall!(
    Height Map=@n7
    Normal Map Intensity=ScalarParameter() {defaultValue=0.5, parameterName=VNR_Normal_Transition}
    Height Map UV Offset
    Coordinates=@n6
    Height Map Channel Selector
   ) {materialFunction=/Engine/Functions/Engine_MaterialFunctions03/Procedurals/NormalFromHeightmap}.Result) {name=Transition Normals, nodeColor=<rgba 0 1 0.1797 1>}
  Transform(None=@n16) {transformSourceType=TRANSFORMSOURCE_World, transformType=TRANSFORM_Tangent}
 ]}
